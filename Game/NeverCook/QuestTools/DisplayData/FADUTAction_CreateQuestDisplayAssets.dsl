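; Editor asset action utility that creates or refreshes quest display data assets for item definitions passed as an asset data array.
; It loads each asset, casts it to FortItemDefinition and calls CreateorRefreshDisplayAssetforItemDefinition on it.
; It combines each call's result into an AllSuccess flag and returns whether every one succeeded.
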(fn RunActionOnAssetDataArray
  (params (AssetDataArray :type Array%20of%20Asset%20Data%20Structures))
  (LocalVariables|SetAllSuccess true)
  (for _array_element AssetDataArray
    (bind _asfort_item_definition (Utilities|Casting|CastToFortItemDefinition (AssetData|GetAsset _array_element)))
    (bind (_outdisplaydataasset _returnvalue) (DisplayData|CreateorRefreshDisplayAssetforItemDefinition _asfort_item_definition))
    (LocalVariables|SetAllSuccess (and _returnvalue (LocalVariables|GetAllSuccess))))
  (return (LocalVariables|GetAllSuccess)))
